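# Additive unlit glow material for vending machine effects, set up for mesh particles and Nanite.
# View-space reflection vectors act as matcap-style UVs into a rainbow crystal texture, tinted by particle color and passed through MF_FX_EmissiveAdjustments.
# A sphere mask around the transformed object position, offset in Z by a dynamic parameter, scales particle alpha.
# Despite its name, BlendInFresnel only blends in a second particle-alpha multiply, with no Fresnel node.
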Material /Game/Effects/Fort_Effects/Gameplay/VendingMachine/M_VendingMachine_Glow
graph {
 properties {
  bUsedWithMeshParticles=true
  bUsedWithNanite=true
  blendMode=BLEND_Additive
  pixelDepthOffsetMode=PDOM_Legacy
  shadingModel=MSM_Unlit
  usedShadingModels=Unlit
 }
 nodes [
  n1 = ParticleColor()
  n2 = WorldPosition()
  n3 = Constant()
  n4 = Add(
   A=TransformPosition(None=ObjectPositionWS()) {transformType=TRANSFORMPOSSOURCE_World}
   B=AppendVector(
    A=AppendVector(A=@n3, B=@n3)
    B=DynamicParameter() {defaultValue=<rgba 1 100 1 1>, paramNames [BlendInFresnel, OffsetMaskZ, Param3, Param4]}.OffsetMaskZ
   )
  )
  n5 = ScalarParameter() {defaultValue=300, parameterName=Radius}
  n6 = ScalarParameter() {parameterName=Hardness}
  n7 = Multiply(A=@n1.A, B=Clamp(None=SphereMask(A=@n2.XYZ, B=@n4, Radius=@n5, Hardness=@n6)))
  LinearInterpolate(
   A=@n7
   B=Multiply(A=@n1.A, B=@n7)
   Alpha=DynamicParameter() {defaultValue=<rgba 1 100 1 1>, paramNames [BlendInFresnel, OffsetMaskZ, Param3, Param4]}.BlendInFresnel
  )
  n8 = Transform(None=ReflectionVectorWS()) {transformSourceType=TRANSFORMSOURCE_World, transformType=TRANSFORM_View}
  n9 = Multiply(A=ConstantBiasScale(None=ComponentMask(None=@n8) {g=true, r=true}), B=Constant2Vector() {g=-1, r=1})
  MaterialFunctionCall(Emissive=Multiply(A=TextureSample(UVs=@n9) {texture=/Game/Environments/Props/Rocks/Textures/T_Props_CrystalEnv_Rainbow}.RGB, B=@n1.RGB)) {materialFunction=&MF_FX_EmissiveAdjustments}
 ]
}
